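# Translucent unlit deferred decal material for the melee heavy attack, with dark decal color and hot orange glow.
# A packed mask texture shapes opacity, which fades with decal lifetime, and drives a sphere-masked two-color glow with optional custom tint.
# The glow falls off sharply through a high power of lifetime opacity.
# A separate branch erodes a masked crack texture over lifetime by animating its exponent, with an optional sphere mask.
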Material /CRD_Melee/FX/HeavyAttack/M_Melee_Decal_HeavyAttack
graph {
 properties {
  bAutomaticallySetUsageInEditor=false
  blendMode=BLEND_Translucent
  materialDomain=MD_DeferredDecal
  pixelDepthOffsetMode=PDOM_Legacy
  shadingModel=MSM_Unlit
  usedShadingModels=Unlit
 }
 nodes [
  n1 = TextureSampleParameter2D() {parameterName=DecalTexture, texture=&/T_Melee_HeavyAttack_Decal_M}
  n2 = LinearInterpolate(Alpha=@n1.B) {constA=0.85, constB=1.2}
  n3 = DecalLifetimeOpacity()
  n4 = ScalarParameter() {defaultValue=6, group=Timestamp, parameterName=Fade Transition Multiplier}
  Multiply(
   A=Multiply(A=Multiply(A=@n1.G, B=@n2), B=ScalarParameter() {defaultValue=1, parameterName=Opacity Multiplider})
   B=Saturate(None=Multiply(A=@n3.Opacity, B=@n4) {constB=4})
  )
  n5 = DecalLifetimeOpacity()
  n6 = Clamp(None=LinearInterpolate(A=ScalarParameter() {defaultValue=-3, parameterName=EmissiveFadeOffset}, Alpha=Saturate(None=@n5.Opacity)) {constA=-7})
  n7 = TextureCoordinate()
  n8 = VectorParameter() {parameterName=EmissiveUV, defaultValue=<rgba 2.5 2.5 0 0>}
  n9 = AppendVector(A=@n8.R, B=@n8.R)
  n10 = TextureSampleParameter2D(UVs=Add(A=Multiply(A=@n7, B=@n9), B=AppendVector(A=@n8.B, B=@n8.B))) {parameterName=EmissiveTex, texture=/Game/Global/Textures/Default/Masks_and_Gradients/Cracks_02}
  n11 = Power(Base=Add(A=@n10.R, B=Multiply(A=@n10.B) {constB=0.3}), Exp=ScalarParameter() {defaultValue=1, parameterName=EmissivePower})
  n12 = TextureCoordinate()
  n13 = Constant() {r=0.5}
  n14 = ScalarParameter() {defaultValue=0.5, parameterName=Radius}
  n15 = ScalarParameter() {parameterName=Hardness}
  n16 = LinearInterpolate(A=@n11, B=Multiply(A=@n11, B=SphereMask(A=@n12, B=@n13, Radius=@n14, Hardness=@n15)), Alpha=ScalarParameter() {parameterName=UseSphereMask})
  n17 = TextureSampleParameter2D() {parameterName=Emissive Tex Mask, texture=&T_Melee_HeavyAttack_Decal_M}
  Multiply(A=Power(Base=@n6), B=Power(Base=Multiply(A=@n16, B=@n17.R), Exp=LinearInterpolate(Alpha=@n6) {constA=14}))
  n18 = TextureSampleParameter2D() {parameterName=Emissive Tex Mask, texture=&T_Melee_HeavyAttack_Decal_M}
  n19 = SphereMask(A=TextureCoordinate(), B=Constant() {r=0.5}, Radius=Constant() {r=0.4}, Hardness=Constant())
  n20 = Multiply(A=@n18.R, B=Multiply(A=@n19, B=@n19))
  n21 = Multiply(A=@n20, B=LinearInterpolate(A=@n18.B, Alpha=@n20))
  n22 = Multiply(
   A=LinearInterpolate(
    A=VectorParameter() {parameterName=Emissive 1, defaultValue=<rgba 1 0.02 0 1>}.RGB
    B=VectorParameter() {parameterName=Emissive 2, defaultValue=<rgba 1.25 0.2975 0 1>}.RGB
    Alpha=@n21
   )
   B=@n21
  )
  n23 = Multiply(A=Desaturation(None=@n22), B=VectorParameter() {parameterName=CustomColor}.RGB)
  n24 = ScalarParameter() {parameterName=UseCustomColor}
  n25 = DecalLifetimeOpacity()
  Multiply(
   A=Multiply(A=LinearInterpolate(A=@n22, B=@n23, Alpha=@n24), B=Power(Base=@n25.Opacity) {constExponent=25})
   B=ScalarParameter() {defaultValue=35, parameterName=Emissive Strebgth}
  )
  ScalarParameter() {parameterName=Metallic}
  ScalarParameter() {parameterName=Specular}
  ScalarParameter() {defaultValue=0.5, parameterName=Roughness}
  VectorParameter() {parameterName=DecalColor, defaultValue=<rgba 0.0052 0.0052 0.0052 1>}
 ]
}
